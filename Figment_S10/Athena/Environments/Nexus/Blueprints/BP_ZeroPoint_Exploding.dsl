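; Exploding Zero Point actor for the Nexus: a sphere volume that grants low gravity to players and vehicles, with bubble visuals and audio.
; Overlaps add or remove gameplay effects; exiting players ignore their next fall damage.
; Clients disable sphere collision and run a distance-adaptive timer positioning edge audio, toggling inside ambience, low-pass filter and random one-shots.
; Construction and Progress OnRep match sphere scale to radius and drive destabilization materials and audio.
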(event EventBeginPlay
  (if (not (Networking|IsServer))
    (Collision|SetCollisionProfileName (Variables|Default|GetZeroRadius) "nocollision" true :!class /Script/Engine.PrimitiveComponent)
    (Utilities|Time|SetTimerbyEvent (EventDispatchers|CreateEvent :!event Custom|Calc%20Closest%20to%20Sphere) (Variables|Default|GetUpdateFrequency))
    (Components|Timeline|SetPlayRate (Variables|BP_ZeroPoint_Exploding|GetApply%20LPF) (/ 1.0 1.0))
    (Audio|Components|Audio|SetLowPassFilterEnabled (Variables|Default|GetEdgeAudio) true :!class /Script/Engine.AudioComponent)))

(event OnComponentBeginOverlap(ZeroRadius)
  (params (OverlappedComponent :type Primitive%20Component%20Object%20Reference) (OtherActor :type Actor%20Object%20Reference) (OtherComp :type Primitive%20Component%20Object%20Reference) (OtherBodyIndex :type Integer) (bFromSweep :type Boolean) (SweepResult :type Hit%20Result%20Structure%20%28by%20ref%29))
  (bind _asplayer_pawn_athena (Utilities|Casting|CastToPlayerPawn_Athena OtherActor)
    (:then
      (Environment|ApplyGlobalEnvironmentGameplayEffecttoActor _asplayer_pawn_athena "/Figment_S10/Athena/Environments/Cattus/GE_LowGravZone.GE_LowGravZone_C"))
    (:CastFailed
      (bind _asfort_athena_vehicle (Utilities|Casting|CastToFortAthenaVehicle OtherActor)
        (:then
          (Environment|ApplyGlobalEnvironmentGameplayEffecttoActor _asfort_athena_vehicle "/Figment_S10/Athena/Environments/Cattus/GE_LowGravZone_Vehicle.GE_LowGravZone_Vehicle_C"))
        (:CastFailed
          (bind _asfort_character_vehicle_ostrich (Utilities|Casting|CastToFortCharacterVehicle_Ostrich OtherActor))
          (Environment|ApplyGlobalEnvironmentGameplayEffecttoActor _asfort_character_vehicle_ostrich "/Figment_S10/Athena/Environments/Cattus/GE_LowGravZone_PawnVehicle.GE_LowGravZone_PawnVehicle_C"))))))

(event OnComponentEndOverlap(ZeroRadius)
  (params (OverlappedComponent :type Primitive%20Component%20Object%20Reference) (OtherActor :type Actor%20Object%20Reference) (OtherComp :type Primitive%20Component%20Object%20Reference) (OtherBodyIndex :type Integer))
  (bind _asplayer_pawn_athena (Utilities|Casting|CastToPlayerPawn_Athena OtherActor)
    (:then
      (Variables|/Script/FortniteGame.FortPawn|SetbIgnoreNextFallingDamage true _asplayer_pawn_athena)
      (GameplayEffects|RemoveActiveEffectswithGrantedTags (Ability|GetAbilitySystemComponent _asplayer_pawn_athena) "(GameplayTags=((TagName=\"Status.LowGravity.LowGravZone\")))"))
    (:CastFailed
      (bind _asfort_athena_vehicle (Utilities|Casting|CastToFortAthenaVehicle OtherActor)
        (:then
          (GameplayEffects|RemoveActiveEffectswithGrantedTags (Ability|GetAbilitySystemComponent _asfort_athena_vehicle) "(GameplayTags=((TagName=\"Status.LowGravity.LowGravZone\")))"))
        (:CastFailed
          (bind _asfort_character_vehicle_ostrich (Utilities|Casting|CastToFortCharacterVehicle_Ostrich OtherActor))
          (GameplayEffects|RemoveActiveEffectswithGrantedTags (Ability|GetAbilitySystemComponent _asfort_character_vehicle_ostrich) "(GameplayTags=((TagName=\"Status.LowGravity.LowGravZone\")))"))))))

(event Custom|Calc%20Closest%20to%20Sphere
  (bind _returnvalue (Transformation|GetActorLocation))
  (bind _returnvalue_2 (Fortnite|GetLocalFortPlayerControllers true true))
  (bind _output (Utilities|Array|Get(acopy) _returnvalue_2))
  (bind _returnvalue_3 (Pawn|GetViewTarget _output))
  (bind _returnvalue_4 (Transformation|GetActorLocation _returnvalue_3))
  (bind _returnvalue_5 (- _returnvalue_4 _returnvalue))
  (bind _returnvalue_6 (Math|Vector|Normalize _returnvalue_5))
  (bind _radius (CallFunction|GetBubbleRadius))
  (bind _returnvalue_7 (* _returnvalue_6 _radius))
  (bind _radius_2 (CallFunction|GetBubbleRadius))
  (bind _playerdistancetocenter (Variables|Default|GetPlayerDistanceToCenter))
  (bind _returnvalue_8 (<= _playerdistancetocenter _radius_2))
  (bind _returnvalue_9 (+ _returnvalue_7 _returnvalue))
  (bind _ambientaudioinside (Variables|Default|GetAmbientAudioInside))
  (Utilities|Time|SetTimerbyEvent (delegate Custom|Calc%20Closest%20to%20Sphere) (Variables|Default|GetUpdateFrequency))
  (Utilities|IsValid _output
    (:Is%20Valid
      (Transformation|SetWorldLocation (Variables|Default|GetEdgeAudio) _returnvalue_9 false false)
      (Variables|Default|SetPlayerDistanceToEdge (Math|Vector|VectorLength (- _returnvalue_4 _returnvalue_9)))
      (Variables|Default|SetPlayerDistanceToCenter (Math|Vector|VectorLength _returnvalue_5))
      (Utilities|FlowControl|Sequence
        (:then_0
          (Variables|Default|SetUpdateFrequency (select (> (Variables|Default|GetPlayerDistanceToEdge) (Variables|/Script/Engine.SoundBase|GetMaxDistance (Variables|/Script/Engine.AudioComponent|GetSound (Variables|Default|GetEdgeAudio)))) 2.0 0.2)))
        (:then_1
          (if (Math|Boolean|NotEqual(Boolean) (Variables|Default|GetPlayerInside) _returnvalue_8)
            (bind _output_get (Variables|Default|SetPlayerInside _returnvalue_8))
            (if _output_get
              (Audio|Components|Audio|FadeIn _ambientaudioinside 1.0 1.0 0.0 "Linear" :!class /Script/Engine.AudioComponent)
              (bind (_direction _alpha) (Timeline|Apply%20LPF Play)
                (:Update
                  (Audio|Components|Audio|SetLowPassFilterFrequency (Variables|Default|GetEdgeAudio) (Math|Float|MapRangeUnclamped (- 1.0 _alpha) 0.0 1.0 3e+02 2e+04) :!class /Script/Engine.AudioComponent))
                (:Finished
                  (switch Utilities|FlowControl|Switch|SwitchonETimelineDirection _direction
                    (:Forward
                      (Utilities|Time|SetTimerbyEvent (delegate Custom|On%20Play%20Amb%20Oneshot) (Math|Random|RandomFloatinRange 5.0 3e+01)))
                    (:Backward
                      (Utilities|Time|SetTimerbyEvent (delegate Custom|On%20Play%20Amb%20Oneshot))))))
              (else
                (Audio|Components|Audio|FadeOut _ambientaudioinside 1.0 0.0 "Linear" :!class /Script/Engine.AudioComponent)
                (Timeline|Apply%20LPF Reverse)))))))))

(event Custom|On%20Play%20Amb%20Oneshot
  (Audio|PlaySoundatLocation "/Figment_S10/Athena/Sounds/Events/Cattus/ZeroPoint/ZeroPoint_Bubble_Inside_Edge_Oneshot_Cue.ZeroPoint_Bubble_Inside_Edge_Oneshot_Cue" (* (Math|Vector|RotateVectorAroundAxis (Math|Vector|MakeVector 0.0) (Math|Random|RandomFloatinRange 0.0 3.6e+02) "0, 0,1.000000") (CallFunction|GetBubbleRadius)))
  (Utilities|Time|SetTimerbyEvent (delegate Custom|On%20Play%20Amb%20Oneshot) (Math|Random|RandomFloatinRange 5.0 3e+01)))

(fn GetBubbleRadius
  (return (Variables|/Script/Engine.SphereComponent|GetSphereRadius (Variables|Default|GetZeroRadius))))

(fn OnRep_Progress
  (CallFunction|UpdateProgress))

(fn UpdateProgress
  (Utilities|FlowControl|Sequence
    (:then_0
      ;[1 1 1] Match sphere visuals to gameplay volume
      (Transformation|SetRelativeScale3D (Variables|Default|GetOuterSphere) (Math|Conversions|ToVector(Float) (/ (Variables|/Script/Engine.SphereComponent|GetSphereRadius (Variables|Default|GetZeroRadius)) 1e+02))))
    (:then_1
      (Rendering|Material|SetScalarParameterValue "/Game/Athena/Environments/Nexus/Materials/FXshader/MPC_Nexus.MPC_Nexus" "Destabilize" (Math|Float|Clamp(Float) (Variables|Default|GetProgress)) :!class /Script/Engine.KismetMaterialLibrary)
      (Rendering|Material|SetScalarParameterValueonMaterials (Variables|Default|GetSM_Nexus_Core) "Progress" (Math|Float|Clamp(Float) (Variables|Default|GetProgress)))
      (Rendering|Material|SetScalarParameterValueonMaterials (Variables|Default|GetSM_Nexus_CracksRays) "WPOExtend" (Math|Float|MapRangeClamped (Variables|Default|GetProgress) 0.1 0.7 5.0 1e+01))
      (Audio|Parameter|SetFloatParameter (Variables|Default|GetAmbientAudio) "Progress" (Variables|Default|GetProgress) :!class /Script/Engine.AudioComponent))))

(fn ConstructionScript
  (CallFunction|UpdateProgress))
