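# Shattered Nexus core material: lit, glossy cracked glass with glowing cracks, an outer Fresnel rim and an inner glow, usable on instanced meshes and particles.
# Layered sines drive flicker, sped up by the MPC_Nexus Destabilize parameter and held constant on low quality and mobile.
# MF_Nex_Core_WPO supplies the inner glow mask and shatter vertex offset, with a static switch adding RiftBeacon offset.
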Material /Figment_S10/Athena/Environments/Nexus/Materials/M_Nex_Core_Shattered
graph {
 properties {
  bUsedWithInstancedStaticMeshes=true
  bUsedWithMeshParticles=true
  bUsedWithParticleSprites=true
  pixelDepthOffsetMode=PDOM_Legacy
  usedShadingModels=Default Lit
 }
 nodes [
  n1 = TextureCoordinate() {uTiling=2, vTiling=2}
  n2 = LinearInterpolate(
   A=TextureSample(UVs=@n1) {texture=/Game/Effects/Fort_Effects/Textures/Tile/T_Glass_Cracks_M}.G
   B=TextureSample() {texture=&T_Glass_Cracks_M}.G
  ) {constAlpha=0.7}
  n3 = VectorParameter() {parameterName=Outer, defaultValue=<rgba 0.1073 0.0986 0.25 1>}
  n4 = ScalarParameter() {defaultValue=0.8, parameterName=FlickerSpeed}
  n5 = Multiply(
   A=Time()
   B=LinearInterpolate(
    A=@n4
    B=Multiply(A=@n4) {constB=2}
    Alpha=CollectionParameter() {collection=/Game/Athena/Environments/Nexus/Materials/FXshader/MPC_Nexus, parameterName=Destabilize}
   )
  )
  n6 = Add(A=Multiply(B=Sine(None=@n5) {period=0.55}), B=Multiply(A=Sine(None=@n5) {period=0.11}) {constB=0.2})
  n7 = Multiply(A=Sine(None=@n5) {period=0.175}) {constB=0.3}
  n8 = Multiply(A=Sine(None=@n5) {period=3.25}, B=Sine(None=@n5) {period=2.325})
  n9 = Constant() {r=0.6}
  n10 = QualitySwitch(Default=Multiply(A=Add(A=@n6, B=@n7), B=ConstantBiasScale(None=@n8)), Low=@n9)
  n11 = FeatureLevelSwitch(Default=@n10, ES2=@n9, ES3_1=@n9, SM4_REMOVED=@n10)
  n12 = Multiply(A=@n11) {constB=0.01}
  n13 = Fresnel(ExponentIn=Add(A=ScalarParameter() {defaultValue=2, parameterName=Fresnel1}, B=@n12)) {baseReflectFraction=0}
  n14 = LinearInterpolate(Alpha=@n11) {constA=0.7}
  n15 = ScalarParameter() {defaultValue=0.4, parameterName=Fresnel2}
  n16 = OneMinus(None=Fresnel(ExponentIn=Add(A=@n12, B=@n15)) {baseReflectFraction=0})
  n17 = Add(A=@n12, B=ScalarParameter() {defaultValue=4, parameterName=Fresnel2Power})
  n18 = MaterialFunctionCall!(Time Multiplier, Lerp Shatter Pulse, Shake, Progress) {materialFunction=/Game/Athena/Environments/Nexus/Materials/MF_Nex_Core_WPO}
  n19 = Multiply(
   A=Multiply(A=Power(Base=@n16, Exp=@n17), B=VectorParameter() {parameterName=Inner, defaultValue=<rgba 0.8023 1.9068 3 1>}.RGB)
   B=Saturate(None=LinearInterpolate(Alpha=@n18.Mask) {constA=1, constB=-0.5})
  )
  Add(
   A=Multiply(A=VectorParameter() {parameterName=CrackGlow, defaultValue=<rgba 2.4794 9.159 16 1>}.RGB, B=Power(Base=@n2))
   B=Add(A=Multiply(A=@n3.RGB, B=@n13), B=Multiply(A=@n14, B=@n19))
  )
  Constant() {r=1}
  MaterialFunctionCall(
   Normal=LinearInterpolate(
    A=TextureSample(UVs=@n1) {samplerType=SAMPLERTYPE_Normal, texture=/Game/Effects/Fort_Effects/Textures/Tile/T_Glass_Cracks_N}.RGB
    B=TextureSample() {samplerType=SAMPLERTYPE_Normal, texture=&T_Glass_Cracks_N}.RGB
   )
   Flatness=ScalarParameter() {defaultValue=1, parameterName=Flatness}
  ) {materialFunction=&FlattenNormal}
  MaterialProxyReplace(Realtime=Constant(), MaterialProxy=Constant3Vector() {constant=<rgba 0.1529 0.451 0.5922 1>})
  ScalarParameter() {defaultValue=0.02, parameterName=Roughness}
  StaticSwitchParameter(
   True=Add(A=@n18.WPOSpread, B=MaterialFunctionCall() {materialFunction=/Game/Environments/Props/RiftBeacon/FX/RiftBeacon_NexusWPO}.Result)
   False=@n18.WPOSpread
  ) {parameterName=Used with RiftBeacon}
 ]
}
