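# Unlit, translucent, two-sided material for glowing light rays on the Nexus shattered core.
# Time-panned sine noise on the direction from the actor center, replaced by a constant on mobile, multiplies a UV gradient to drive color and opacity.
# A sharper gradient power adds a hot base glow, while view-angle and gradient-varied depth fades soften opacity.
# A world offset stretches vertices away from the actor center by inverse V, scaled by WPOExtend.
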Material /Figment_S10/Athena/Environments/Nexus/Materials/M_Nex_Core_Shattered_Rays
graph {
 properties {
  bUsedWithParticleSprites=true
  blendMode=BLEND_Translucent
  pixelDepthOffsetMode=PDOM_Legacy
  shadingModel=MSM_Unlit
  twoSided=true
  usedShadingModels=Unlit
 }
 nodes [
  n1 = Normalize(VectorInput=Subtract(A=WorldPosition().XYZ, B=ActorPositionWS()))
  n2 = Multiply(A=Time(), B=Multiply(A=ScalarParameter() {defaultValue=0.1, parameterName=PanSpeed}, B=Constant3Vector() {constant=<rgba 0.8 0.6 1 0>}))
  n3 = ConstantBiasScale(None=Sine(None=Add(A=@n1, B=@n2)) {period=0.15})
  n4 = LinearInterpolate(Alpha=Desaturation(None=Power(Base=@n3))) {constA=0.6}
  n5 = Constant() {r=1}
  n6 = ComponentMask(None=TextureCoordinate()) {g=true}
  n7 = Multiply(
   A=FeatureLevelSwitch(Default=@n4, ES2=@n5, ES3_1=@n5, SM4_REMOVED=@n4)
   B=Power(Base=@n6, Exp=ScalarParameter() {defaultValue=3, parameterName=GradientPower}) {constExponent=3}
  )
  n8 = ScalarParameter() {defaultValue=14, parameterName=HotBasePower}
  Add(
   A=LinearInterpolate(
    A=VectorParameter() {parameterName=TopColor, defaultValue=<rgba 0.4695 0.3583 3 0>}.RGB
    B=VectorParameter() {parameterName=Inner, defaultValue=<rgba 1.3371 3.1779 5 1>}.RGB
    Alpha=@n7
   )
   B=Multiply(
    A=Power(Base=@n6, Exp=@n8) {constExponent=12}
    B=VectorParameter() {parameterName=HotBaseColor, defaultValue=<rgba 23.1771 46.3256 50 0>}.RGB
   )
  )
  n9 = MaterialFunctionCall("FadeSensitivity (optional)"=ScalarParameter() {defaultValue=2, parameterName=AngleFade}) {materialFunction=&FadeBasedOnViewAngle}
  n10 = Multiply(A=@n9.Result, B=@n9.Result)
  n11 = ScalarParameter() {defaultValue=1, parameterName=Opacity}
  n12 = ScalarParameter() {parameterName=InnerDepthFade}
  n13 = ScalarParameter() {defaultValue=800, parameterName=OuterDepthFade}
  Multiply(A=@n7, B=DepthFade(Opacity=Multiply(A=@n10, B=@n11), FadeDistance=LinearInterpolate(A=@n12, B=@n13, Alpha=@n6) {constA=40, constB=0}))
  n14 = ComponentMask(None=TextureCoordinate()) {g=true}
  n15 = WorldPosition()
  n16 = ActorPositionWS()
  Multiply(A=Multiply(A=OneMinus(None=@n14), B=Subtract(A=@n15.XYZ, B=@n16)), B=ScalarParameter() {defaultValue=3, parameterName=WPOExtend})
 ]
}
