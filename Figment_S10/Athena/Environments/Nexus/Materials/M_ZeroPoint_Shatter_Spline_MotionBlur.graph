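# Translucent, unlit, two-sided particle material for motion-blurred spline shards in the Nexus Zero Point shatter effect.
# SplineThicken sets line width and a V-masked offset stretches vertices toward the actor position, both driven by dynamic parameters.
# Color and opacity blend two half-offset samples of one blur texture using MotionBlur plus layered-sine flicker.
# Particle color tints the green-boosted result, and a power curve shapes opacity.
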Material /Figment_S10/Athena/Environments/Nexus/Materials/M_ZeroPoint_Shatter_Spline_MotionBlur
graph {
 properties {
  bUsedWithMeshParticles=true
  bUsedWithParticleSprites=true
  blendMode=BLEND_Translucent
  pixelDepthOffsetMode=PDOM_Legacy
  previewMesh=/Figment_S10/Athena/Environments/Nexus/Meshes/SM_ZeroPoint_Shatter_Burst_Spline
  shadingModel=MSM_Unlit
  twoSided=true
  usedShadingModels=Unlit
 }
 nodes [
  n1 = DynamicParameter() {defaultValue=<rgba 1 1 0.5 1>, paramNames [MotionBlur,WidthMult,InnerStretch,BlurFlicker]}
  n2 = Multiply(A=@n1.WidthMult, B=ScalarParameter() {defaultValue=20, parameterName=BaseSplineWidth})
  n3 = ComponentMask(None=TextureCoordinate()) {g=true}
  n4 = Subtract(A=ActorPositionWS(), B=WorldPosition().XYZ)
  Add(
   A=MaterialFunctionCall(WidthBase=@n2, WidthTip=@n2) {materialFunction=&SplineThicken}.WorldPositionOffset
   B=Multiply(A=@n1.InnerStretch, B=Multiply(A=@n3, B=@n4))
  )
  MaterialFunctionCall() {materialFunction=&DebugTimeSine}
  n5 = ParticleColor()
  n6 = TextureCoordinate() {coordinateIndex=1}
  n7 = TextureSample(UVs=@n6) {texture=&/T_ZeroPoint_Shards_Blur}
  n8 = TextureSample(UVs=Add(A=@n6, B=Constant2Vector() {g=0.5})) {texture=&T_ZeroPoint_Shards_Blur}
  n9 = Multiply(A=Add(A=Time(), B=@n1.BlurFlicker))
  n10 = Sine(None=@n9) {period=0.55}
  n11 = Sine(None=@n9) {period=0.11}
  n12 = Add(A=Add(A=Multiply(B=@n10), B=Multiply(A=@n11) {constB=0.2}), B=Multiply(A=Sine(None=@n9) {period=0.175}) {constB=0.3})
  n13 = ConstantBiasScale(None=Multiply(A=Sine(None=@n9) {period=3.25}, B=Sine(None=@n9) {period=2.325}))
  n14 = Multiply(A=Multiply(A=Multiply(A=@n12, B=@n13)) {constB=0.25}, B=@n1.BlurFlicker)
  n15 = Multiply(A=@n14) {constB=0.4}
  n16 = LinearInterpolate(A=@n7.RGB, B=@n8.RGB, Alpha=Saturate(None=Add(A=@n15, B=@n1.MotionBlur)))
  n17 = ComponentMask(None=@n16) {g=true}
  Multiply(A=@n5.RGB, B=Add(A=Multiply(A=@n17, B=@n16), B=@n16))
  n18 = Saturate(None=Add(A=@n14, B=@n1.MotionBlur))
  Power(
   Base=Multiply(A=@n5.A, B=LinearInterpolate(A=@n7.A, B=@n8.A, Alpha=@n18))
   Exp=ScalarParameter() {defaultValue=2, parameterName=OpacityPower}
  )
  ScalarParameter() {defaultValue=0.8, parameterName=FlickerSpeed}
 ]
}
